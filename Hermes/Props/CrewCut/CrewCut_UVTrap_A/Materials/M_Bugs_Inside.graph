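# Masked, Default Lit material for Niagara sprites that draws bugs, apparently inside the CrewCut UV Trap A prop.
# It samples the alpha of a baked single-wasp texture from the insect swarm effects and uses it to lerp up to a scalar parameter.
# Saturate clamps that boosted result, hardening the wasp silhouette.
# A Particle Color node is also present, and the texture's RGB is unused.
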Material /Hermes/Props/CrewCut/CrewCut_UVTrap_A/Materials/M_Bugs_Inside
graph {
 properties {bAutomaticallySetUsageInEditor=false, bUsedWithNiagaraSprites=true, blendMode=BLEND_Masked, usedShadingModels=Default Lit}
 nodes [
  ParticleColor()
  Saturate(None=LinearInterpolate(
    B=ScalarParameter() {defaultValue=10}
    Alpha=TextureSample() {texture=/Game/Effects/Fort_Effects/Effects/Environment/Bugs/BugBiome/Materials/T_AI_Wildlife_InsectSwarm_SingleWasp_Baked_V4_1}.A
   ) {constB=2})
 ]
}
