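# Translucent, unlit, two-sided glow material for the CrewCut UV trap prop.
# Emissive blends two blue-cyan colors along a UV V power gradient, then passes through MF_FX_EmissiveAdjustments.
# A faint V gradient mask pulses as a time-driven sine varies its exponent, phase-offset by actor world position so instances desynchronize.
# An inverted Fresnel raised to a power fades that mask toward silhouette edges, softening the glow.
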Material /Hermes/Props/CrewCut/CrewCut_UVTrap_A/Materials/M_CrewCut_Trap_Glow
graph {
 properties {
  bAutomaticallySetUsageInEditor=false
  bOutputTranslucentVelocity=true
  blendMode=BLEND_Translucent
  shadingModel=MSM_Unlit
  twoSided=true
  usedShadingModels=Unlit
 }
 nodes [
  n1 = ComponentMask(None=TextureCoordinate()) {g=true}
  n2 = ScalarParameter() {defaultValue=8, parameterName=gradient Color}
  MaterialFunctionCall(Emissive=LinearInterpolate(
    A=VectorParameter() {defaultValue=<rgba 0 0.5 1 1>}.RGB
    B=VectorParameter() {parameterName=Param_1, defaultValue=<rgba 0 0.9587 2 1>}.RGB
    Alpha=Power(Base=@n1, Exp=@n2)
   )) {materialFunction=&MF_FX_EmissiveAdjustments}
  n3 = ComponentMask(None=TextureCoordinate()) {g=true}
  n4 = MaterialFunctionCall(Float3=ActorPositionWS()) {materialFunction=&BreakOutFloat3Components}
  n5 = Add(A=@n4.R, B=@n4.R)
  n6 = Add(A=Time(), B=Frac(None=Add(A=@n5, B=@n4.B)))
  n7 = LinearInterpolate(A=ScalarParameter() {defaultValue=5, parameterName=Opacity Gradient}, Alpha=Add(A=Sine(None=@n6))) {constB=6}
  n8 = Fresnel() {exponent=2}
  Multiply(
   A=Multiply(A=Power(Base=@n3, Exp=@n7), B=ScalarParameter() {defaultValue=0.05, parameterName=Param_1})
   B=Power(Base=OneMinus(None=@n8), Exp=ScalarParameter() {defaultValue=8})
  )
 ]
}
