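# Niagara effect for the CrewCut UV trap prop that shows a swarm of bugs circling the system's position.
# A dense emitter spawns small dark sprites with flipbook animation, while a sparse emitter spawns tiny glowing meshes.
# Both use RotateAroundPoint with random axis, radius, phase and rate to form orbits.
# The meshes orient along their orbit direction and pulse HDR color through color curves and an age-driven waveform.
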NiagaraSystem /Hermes/Props/CrewCut/CrewCut_UVTrap_A/Meshes/NS_CrewCut_Bugs
dependencies {
 usedDataInterfaces [
  &NiagaraDataInterfaceCurve
  &NiagaraDataInterfaceSpriteRendererInfo
  &NiagaraDataInterfaceColorCurve
 ]
 usedDynamicInputs []
 usedModules [&SystemState]
}
emitters {
 Minimal {
  emitterUpdateScript [
   &EmitterState()
   &SpawnRate(SpawnRate=50, Use Spawn Probability=false)
  ]
  particleSpawnScript [&InitializeParticle(
    Color=<rgba 0.1 0.1 0.1 1>
    Lifetime=5
    Sprite Size Mode=::NewEnumerator1
    Uniform Sprite Size Max=4
    Uniform Sprite Size Min=2
    UsePositionOffset=false
    Write Material Random=true
   )]
  particleUpdateScript [
   &ParticleState()
   &ScaleSpriteSize(Uniform Curve Sprite Scale=DataInterface() {propertyValues {Curve {keys [{tangentMode=RCTM_User, time=0.0066, value=-0.0078}, <at 0.5 1>, {tangentMode=RCTM_User, time=1, value=0}]}}})
   &RotateAroundPoint(
    Derive the Rotation Axis=false
    Radius:NiagaraFloat=&RandomRangeFloat(Minimum=15, Maximum=35)
    Rotation Axis:Vector3f=&RandomRangeVector(Minimum=<-1 -1 -1>, Maximum=<1 1 1>)
    Rotation Center:NiagaraPosition=/Niagara/DynamicInputs/Helpers/SimulationPosition()
    Rotation Phase:NiagaraFloat=&RandomRangeFloat(Minimum=0, Maximum=1)
    Rotation Rate:NiagaraFloat=&RandomRangeFloat(Minimum=0.2, Maximum=0.25)
   )
   &SubUVAnimation(
    Play Rate=50
    "Sprite Renderer "=DataInterface() {propertyValues {SpriteRenderer=&self:Minimal_0.NiagaraSpriteRendererProperties_0}}
    SubUV Animation Mode:ENiagara_SubUVLookupModeV2=::NewEnumerator3
    UseEndFrame=false
    UsePlayRateScale=false
    UseStartFrame=false
   )
  ]
  rendererClasses [&NiagaraSpriteRendererProperties]
  renderers [{rendererClass=&NiagaraSpriteRendererProperties}]
 }
 Minimal001 {
  emitterUpdateScript [
   &EmitterState()
   &SpawnRate(SpawnRate=2, Use Spawn Probability=false)
  ]
  particleSpawnScript [&InitializeParticle(
    Color=<rgba 357.5332 0 500 1>
    Lifetime=5
    Mesh Scale Mode=::NewEnumerator1
    Mesh Uniform Scale Max=0.02
    Mesh Uniform Scale Min=0.01
    Sprite Size Mode=::NewEnumerator1
    Uniform Sprite Size Max=1
    Uniform Sprite Size Min=0.5
    UsePositionOffset=false
    Write Material Random=true
   )]
  particleUpdateScript [
   &ParticleState()
   &ScaleSpriteSize(Uniform Curve Sprite Scale=DataInterface() {propertyValues {Curve {keys [{tangentMode=RCTM_User, time=0.0066, value=-0.0078}, <at 0.5 1>, {tangentMode=RCTM_User, time=1, value=0}]}}})
   &RotateAroundPoint(
    Derive the Rotation Axis=false
    Radius:NiagaraFloat=&RandomRangeFloat(Minimum=15, Maximum=35)
    Rotation Axis:Vector3f=&RandomRangeVector(Minimum=<-1 -1 -1>, Maximum=<1 1 1>)
    Rotation Center:NiagaraPosition=&SimulationPosition()
    Rotation Phase:NiagaraFloat=&RandomRangeFloat(Minimum=0, Maximum=1)
    Rotation Rate:NiagaraFloat=&RandomRangeFloat(Minimum=0.2, Maximum=0.25)
   )
   &SubUVAnimation(
    Play Rate=50
    "Sprite Renderer "=DataInterface() {propertyValues {SpriteRenderer=&self:Minimal001_0.NiagaraSpriteRendererProperties_0}}
    SubUV Animation Mode:ENiagara_SubUVLookupModeV2=::NewEnumerator3
    UseEndFrame=false
    UsePlayRateScale=false
    UseStartFrame=false
   )
   /SetVariables(Particles.MeshOrientation=&MakeQuaternion(
     Quaternion From=3
     Axis=<1 0 0>
     Angle=0
     From=<0 -1 0>
     To=$Output.RotateAroundPoint.LocalForwardVector
     Angle Type=::NewEnumerator0
     Normalize Vectors=true
    )) {isSetParameters=true}
   &ScaleColor(
    Linear Color Curve=DataInterface() {propertyValues {
      AlphaCurve {keys [<at 0 1>,<at 1 0>]}
      BlueCurve {keys [<at 0.0097 0>,<at 0.488 1>,<at 1 0>]}
      GreenCurve {keys [<at 0.0097 0>,<at 0.488 1>,<at 1 0>]}
      RedCurve {keys [<at 0.0097 0>,<at 0.488 1>,<at 1 0>]}
     }}
    Scale Mode=::NewEnumerator2
   )
   &ScaleColor(
    Scale RGB=&Lerp_Vector(
     A=<0 0 0>
     B=<50 50 1>
     Alpha=&Waveform(Global Phase=$Particles.Age, "Waveform [1]"=::NewEnumerator0, "[1] Phase Scale"=0.001, "[1] Frequency"=1, "[1] Amplitude Min/Max"=<0 1>)
    )
    ScaleA=true
    ScaleRGB=true
   ) {module=ScaleColor001}
  ]
  rendererClasses [&NiagaraMeshRendererProperties]
  renderers [{rendererClass=&NiagaraMeshRendererProperties, rendererIndex=0}]
 }
}
summary {
 systemName=NS_CrewCut_Bugs
 emitters [
  {emitterName=Minimal, rendererClasses [&NiagaraSpriteRendererProperties]}
  {emitterName=Minimal001, rendererClasses [&NiagaraMeshRendererProperties]}
 ]
 userVariables []
}
